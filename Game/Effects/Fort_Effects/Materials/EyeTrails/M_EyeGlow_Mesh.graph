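# Translucent unlit material for glowing eye effects on skeletal meshes and mesh particles.
# Emissive comes from EyeColor or particle color with optional color Fresnel, through MF_FX_EmissiveAdjustments.
# Opacity combines optional Fresnel (vertex or texture normals), depth fade and alpha; an unconnected feature-level switch would skip depth fade on mobile.
# An optional time-sine jiggle, scaled by a dynamic parameter, drives world position offset, and BaseFN_CharacterOpacity wraps the final attributes.
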Material /Game/Effects/Fort_Effects/Materials/EyeTrails/M_EyeGlow_Mesh
graph {
 properties {
  bUseMaterialAttributes=true
  bUsedWithMeshDeformer=true
  bUsedWithMeshParticles=true
  bUsedWithNiagaraMeshParticles=true
  bUsedWithSkeletalMesh=true
  blendMode=BLEND_Translucent
  pixelDepthOffsetMode=PDOM_Legacy
  shadingModel=MSM_Unlit
  usedShadingModels=Unlit
 }
 nodes [
  n1 = TextureSampleParameter2D() {parameterName=Texture-Normal, samplerType=SAMPLERTYPE_Normal, texture=/Game/Missions/Primary/Starlight/Textures/T_Dungeon_Totem_N}
  n2 = StaticSwitchParameter(True=Transform(None=@n1.RGB), False=VertexNormalWS()) {parameterName=UseNormalTexture?}
  n3 = StaticSwitchParameter(
   True=MaterialFunctionCall(
    Brightness=ScalarParameter() {defaultValue=0.95, parameterName=Fresnel Brightness}
    Exponent=ScalarParameter() {defaultValue=16, parameterName=Fresnel Exponent}
    Vertex / Pixel Normal=@n2
   ) {materialFunction=&MF_Fresnel_FX}.Fresnel
   False=Constant() {r=1}
  ) {defaultValue=true, parameterName=UseOpacityFresnel?}
  n4 = DepthFade(Opacity=@n3, FadeDistance=ScalarParameter() {defaultValue=5, parameterName=DepthFade}) {fadeDistanceDefault=1000}
  FeatureLevelSwitch(Default=@n4, ES2=@n3, ES3_1=@n3, SM4_REMOVED=@n4, SM5=@n4, SM6=@n4)
  n5 = MaterialFunctionCall(
   Brightness=ScalarParameter() {defaultValue=0.95, parameterName=ColorFresnel Brightness}
   Exponent=ScalarParameter() {defaultValue=16, parameterName=ColorFresnel Exponent}
   Vertex / Pixel Normal=@n2
  ) {materialFunction=&MF_Fresnel_FX}
  n6 = ParticleColor()
  n7 = VectorParameter() {group=Eye Color, parameterName=EyeColor, defaultValue=<rgba 10.0282 1.6485 20 1>}
  n8 = StaticSwitchParameter(
   True=StaticSwitchParameter(True=Multiply(A=@n6.RGB, B=@n5.Fresnel), False=@n6.RGB) {parameterName=UseColorFresnel?}
   False=Multiply(A=@n7.RGB, B=ScalarParameter() {defaultValue=1, group=Eye Color, parameterName=EyeColor Multiplier})
  ) {parameterName=UseParticleColor?}
  n9 = Multiply(A=@n4, B=@n6.A)
  n10 = Multiply(A=Multiply(A=@n4, B=ScalarParameter() {defaultValue=1, parameterName=Alpha}), B=@n7.A)
  n11 = Multiply(A=Time(), B=ScalarParameter() {defaultValue=12, parameterName=WPOJIggleSpeed})
  n12 = Multiply(
   A=Sine(None=Frac(None=@n11))
   B=Multiply(
    A=VectorParameter() {parameterName=WPOJIggle Intensity, defaultValue=<rgba 0 0 0 1>}.RGB
    B=DynamicParameter() {paramNames [WPOJIggle-Intensity, Param2, Param3, Param4]}.WPOJIggle-Intensity
   )
  )
  n13 = Constant()
  MaterialFunctionCall!(Material=MakeMaterialAttributes!(
    BaseColor
    Metallic
    Specular
    Roughness
    Anisotropy
    EmissiveColor=MaterialFunctionCall(Emissive=@n8) {materialFunction=&MF_FX_EmissiveAdjustments}.Result
    Opacity=StaticSwitchParameter(True=@n9, False=@n10) {parameterName=UseParticleColor?}
    OpacityMask
    Normal
    Tangent
    WorldPositionOffset=StaticSwitchParameter(True=@n12, False=@n13) {parameterName=Use_WPO-JIggle?}
    SubsurfaceColor
    ClearCoat
    ClearCoatRoughness
    AmbientOcclusion
    Refraction
    CustomizedUVs_0
    CustomizedUVs_1
    CustomizedUVs_2
    CustomizedUVs_3
    CustomizedUVs_4
    CustomizedUVs_5
    CustomizedUVs_6
    CustomizedUVs_7
    PixelDepthOffset
    ShadingModel
    Displacement
   )) {materialFunction=/Game/Packages/Fortress_SharedMaterials/Base_Material_Functions/BaseCharacterShaderFunctions/BaseFN_CharacterOpacity}
 ]
}
