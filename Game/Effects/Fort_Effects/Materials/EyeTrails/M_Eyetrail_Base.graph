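# Unlit, two-sided translucent base material for glowing eye trails on skeletal and Niagara meshes.
# A static switch picks procedural noise flames or a scrolling texture for opacity, shaped by UV edge gradients, view-angle fade and depth fade.
# Emissive blends eye color with optional second color, eye-edge glow, HDR remap and particle color.
# World position offset adds rear-weighted noise wiggle, bends, pushes and camera offset, before a shared character-opacity function wraps the output.
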Material /Game/Effects/Fort_Effects/Materials/EyeTrails/M_Eyetrail_Base
graph {
 properties {
  bUseMaterialAttributes=true
  bUsedWithMeshDeformer=true
  bUsedWithNiagaraMeshParticles=true
  bUsedWithSkeletalMesh=true
  blendMode=BLEND_Translucent
  pixelDepthOffsetMode=PDOM_Legacy
  shadingModel=MSM_Unlit
  twoSided=true
  usedShadingModels=Unlit
 }
 nodes [
  n1 = TextureCoordinate() {vTiling=0.5}
  n2 = ComponentMask(None=VectorParameter() {group=FlameShape, parameterName=Flames_UVScale, sortPriority=3, defaultValue=<rgba 0.6 0.4 0 0>}.RGB) {g=true, r=true}
  n3 = Panner(Coordinate=Add(A=Multiply(A=@n1, B=@n2), B=ScalarParameter() {group=FlameShape, parameterName=Flame_UVOffset, sortPriority=4})) {speedX=-0.5, speedY=1}
  n4 = Add(A=Multiply(A=@n2, B=TextureCoordinate() {uTiling=4}), B=ScalarParameter() {group=FlameShape, parameterName=Flame_UVOffset, sortPriority=4})
  n5 = Multiply(A=Time(), B=ScalarParameter() {defaultValue=2, group=FlameShape, parameterName=Flame_Speed, sortPriority=5})
  n6 = Add(
   A=MaterialFunctionCall(In=TextureSample(UVs=@n3) {texture=&T_NoisePack_01}.G, Contrast=Constant() {r=0.2}) {materialFunction=&CheapContrast}.Result
   B=TextureSample(UVs=Panner(Coordinate=@n4, Time=@n5) {speedX=0.2, speedY=1}) {texture=&T_NoisePack_01}.G
  )
  n7 = Multiply(A=@n2, B=TextureCoordinate() {uTiling=6, vTiling=1.5})
  n8 = ScalarParameter() {group=FlameShape, parameterName=Flame_UVOffset, sortPriority=4}
  n9 = TextureSample(UVs=Panner(Coordinate=Add(A=@n7, B=@n8), Time=@n5) {speedX=-0.2, speedY=1}) {texture=&T_NoisePack_01}
  n10 = MaterialFunctionCall() {materialFunction=&LinearGradient}
  n11 = ScalarParameter() {defaultValue=2, group=FlameShape, parameterName=FireLine, sortPriority=2}
  n12 = Power(Base=Add(A=Multiply(A=@n6, B=@n9.G), B=Power(Base=@n10.VGradient, Exp=@n11))) {constExponent=0.5}
  n13 = ScalarParameter() {defaultValue=0.5, group=FlameShape, parameterName=HardEdgeness, sortPriority=1}
  n14 = Multiply(
   A=TextureCoordinate() {uTiling=0.5, vTiling=0.1}
   B=ComponentMask(None=VectorParameter() {group=Alpha, parameterName=TextureTileUV, sortPriority=5, defaultValue=<rgba 1 1 0 0>}.RGB) {g=true, r=true}
  )
  n15 = LinearInterpolate(B=VertexColor().R, Alpha=ScalarParameter() {defaultValue=1, group=Alpha, parameterName=Use Vertex Color Offset, sortPriority=6})
  n16 = Time()
  n17 = ScalarParameter() {defaultValue=0.5, group=Alpha, parameterName=Alpha_Speed, sortPriority=6}
  n18 = TextureSampleParameter2D(UVs=Panner(Coordinate=Add(A=@n14, B=@n15), Time=Multiply(A=@n16, B=@n17)) {speedX=-0.1, speedY=1}) {
   group=Alpha
   parameterName=Alpha Texture
   samplerSource=SSM_Wrap_WorldGroupSettings
   sortPriority=3
   texture=&T_Inky_Smoke_Tile
  }
  n19 = StaticSwitchParameter(
   True=Clamp(None=MaterialFunctionCall(In=@n12, Contrast=@n13) {materialFunction=&CheapContrast}.Result)
   False=StaticComponentMaskParameter(None=AppendVector(A=@n18.RGB, B=@n18.RGB)) {defaultG=true, group=Alpha, parameterName=Alpha Mask, sortPriority=4}
  ) {group=Flame Based, parameterName=Use Flame Based Trails}
  n20 = MaterialFunctionCall() {materialFunction=&LinearGradient}
  n21 = OneMinus(None=@n20.VGradient)
  n22 = ScalarParameter() {defaultValue=1.5, group=Alpha, parameterName=RearEdgeAlphaContrast, sortPriority=7}
  n23 = Clamp(None=Multiply(
    A=Power(Base=@n21, Exp=@n22) {constExponent=1.5}
    B=ScalarParameter() {defaultValue=10, group=Alpha, parameterName=RearEdgeAlphaIntensity, sortPriority=8}
   ) {constB=10})
  n24 = Power(Base=@n20.VGradient, Exp=ScalarParameter() {defaultValue=3, group=Alpha, parameterName=EyeEdgeAlphaContrast, sortPriority=9}) {constExponent=3}
  n25 = ScalarParameter() {defaultValue=1, group=Alpha, parameterName=EyeEdgeAlphaIntensity, sortPriority=10}
  n26 = Multiply(A=@n23, B=Clamp(None=Multiply(A=@n24, B=@n25)))
  n27 = MaterialFunctionCall() {materialFunction=&LinearGradient}
  n28 = Power(
   Base=OneMinus(None=@n27.UGradient)
   Exp=ScalarParameter() {defaultValue=1.5, group=Alpha, parameterName=TopEdgeAlphaContrast, sortPriority=11}
  ) {constExponent=1.5}
  n29 = ScalarParameter() {defaultValue=10, group=Alpha, parameterName=TopEdgeAlphaIntensity, sortPriority=12}
  n30 = Power(Base=@n27.UGradient, Exp=ScalarParameter() {defaultValue=3, group=Alpha, parameterName=BottomEdgeAlphaContrast, sortPriority=13}) {constExponent=3}
  n31 = ScalarParameter() {defaultValue=1, group=Alpha, parameterName=BottomEdgeAlphaIntensity, sortPriority=14}
  n32 = Multiply(A=Clamp(None=Multiply(A=@n28, B=@n29) {constB=10}), B=Clamp(None=Multiply(A=@n30, B=@n31)))
  n33 = ScalarParameter() {defaultValue=2, group=Alpha, parameterName=FadeViewAngle, sortPriority=2}
  n34 = Multiply(
   A=@n19
   B=Multiply(
    A=Multiply(A=@n26, B=@n32)
    B=MaterialFunctionCall("FadeSensitivity (optional)"=@n33) {materialFunction=&FadeBasedOnViewAngle}.Result
   )
  )
  n35 = DepthFade(Opacity=@n34, FadeDistance=ScalarParameter() {defaultValue=5, group=Alpha, parameterName=DepthFade, sortPriority=1})
  FeatureLevelSwitch(Default=@n35, ES2=@n34, ES3_1=@n34, SM4_REMOVED=@n35, SM5=@n35, SM6=@n35)
  n36 = ParticleColor()
  n37 = Multiply(
   A=VectorParameter() {group=Color, parameterName=EyeColor, sortPriority=1, defaultValue=<rgba 10.0282 1.6485 20 1>}.RGB
   B=ScalarParameter() {defaultValue=1, group=Color, parameterName=EyeColor Multiplier}
  )
  n38 = StaticSwitchParameter(True=LinearInterpolate(A=@n37, B=VectorParameter() {group=Color, parameterName=Color_High, sortPriority=2}.RGB, Alpha=@n19), False=@n37) {group=Color, parameterName=Use Gradient Controlled 2 Color System}
  n39 = Multiply(A=ScalarParameter() {defaultValue=10, group=Bright Edge, parameterName=BrightEdge_Brightness, sortPriority=1}, B=@n38)
  n40 = MaterialFunctionCall() {materialFunction=&LinearGradient}
  n41 = ScalarParameter() {defaultValue=10, group=Bright Edge, parameterName=BrightEdgePower, sortPriority=3}
  n42 = Clamp(None=Multiply(
    A=Power(Base=@n40.VGradient, Exp=@n41)
    B=ScalarParameter() {defaultValue=1, group=Bright Edge, parameterName=BrightEdgeMult, sortPriority=2}
   ))
  n43 = StaticSwitchParameter(True=Add(A=Multiply(A=@n39, B=@n42), B=@n38), False=@n38) {group=Bright Edge, parameterName=UseEyeEdgeGlow}
  n44 = StaticSwitchParameter(
   True=MaterialFunctionCall(
    LDR Input=Clamp(None=@n43)
    HDR Tint=@n43
    HDR Intensity Multiplier=ScalarParameter() {defaultValue=1, group=HDR, parameterName=HDR Intensity, sortPriority=1}
   ) {materialFunction=&DeriveHDRfromLDR}.RemappedRange
   False=@n43
  ) {group=HDR, parameterName=UseHDRfromLDR}
  n45 = LinearInterpolate(
   A=LinearInterpolate(A=@n44, B=Multiply(A=@n44, B=@n36.RGB), Alpha=ScalarParameter() {group=Color, parameterName=Use Particle Color})
   B=@n36.RGB
   Alpha=ScalarParameter() {parameterName=Use ONLY Particle Color}
  )
  n46 = VectorParameter() {group=Color, parameterName=EyeColor, sortPriority=1, defaultValue=<rgba 10.0282 1.6485 20 1>}
  n47 = Multiply(
   A=LinearInterpolate(A=@n35, B=Multiply(A=@n36.A, B=@n35), Alpha=ScalarParameter() {group=Alpha, parameterName=Use Particle Color Alpha})
   B=ScalarParameter() {defaultValue=1, group=Alpha, parameterName=Master_Alpha}
  )
  n48 = MaterialFunctionCall(Offset Amount=ScalarParameter() {group=WPO, parameterName=Camera Offset}, World Position=WorldPosition().XYZ) {materialFunction=/Engine/Functions/Engine_MaterialFunctions02/WorldPositionOffset/CameraOffset}
  n49 = Multiply(
   A=TextureCoordinate() {uTiling=0.5, vTiling=0.5}
   B=ComponentMask(None=VectorParameter() {group=WPO, parameterName=WiggleUV, sortPriority=4, defaultValue=<rgba 1 1 0 0>}.RGB) {g=true, r=true}
  )
  n50 = VertexColor()
  n51 = Time()
  n52 = ScalarParameter() {defaultValue=0.5, group=WPO, parameterName=WPO_Speed, sortPriority=5}
  n53 = TextureSampleParameter2D(UVs=Panner(Coordinate=Add(A=@n49, B=@n50.R), Time=Multiply(A=@n51, B=@n52)) {speedX=0.05, speedY=1}) {
   group=WPO
   parameterName=WPO Texture
   samplerSource=SSM_Wrap_WorldGroupSettings
   sortPriority=1
   texture=&T_NoisePack_01
  }
  n54 = AppendVector(A=@n53.RGB, B=@n53.RGB)
  n55 = ScalarParameter() {defaultValue=2, group=WPO, parameterName=WiggleAmount, sortPriority=3}
  n56 = Transform(None=Multiply(A=@n23, B=Multiply(A=@n54, B=@n55))) {transformSourceType=TRANSFORMSOURCE_Local}
  n57 = MaterialFunctionCall() {materialFunction=&LinearGradient}
  n58 = Multiply(
   A=Power(Base=OneMinus(None=@n57.VGradient), Exp=ScalarParameter() {defaultValue=4, group=WPO_Shape, parameterName=Bend_Up_Contrast, sortPriority=1}) {constExponent=4}
   B=Transform(None=Constant3Vector() {constant=<rgba 0 0 1 0>}) {transformSourceType=TRANSFORMSOURCE_Local}
  )
  n59 = ScalarParameter() {group=WPO_Shape, parameterName=Bend_Mesh_Up}
  n60 = Power(
   Base=OneMinus(None=MaterialFunctionCall() {materialFunction=&LinearGradient}.VGradient)
   Exp=ScalarParameter() {defaultValue=4, group=WPO_Shape, parameterName=Bend_Foward_Contrast, sortPriority=3}
  ) {constExponent=4}
  n61 = Transform(None=Constant3Vector() {constant=<rgba 0 1 0 0>}) {transformSourceType=TRANSFORMSOURCE_Local}
  n62 = Add(
   A=Add(
    A=StaticComponentMaskParameter(None=@n56) {defaultB=true, defaultG=true, defaultR=true, group=WPO, parameterName=WPO Mask, sortPriority=2}
    B=Multiply(A=@n58, B=@n59)
   )
   B=Multiply(A=Multiply(A=@n60, B=@n61), B=ScalarParameter() {group=WPO_Shape, parameterName=Bend_Mesh_Foward, sortPriority=2})
  )
  n63 = Multiply(
   A=Transform(None=Constant3Vector() {constant=<rgba 0 1 0 0>}) {transformSourceType=TRANSFORMSOURCE_Local}
   B=ScalarParameter() {group=WPO_Positioning, parameterName=Push_Mesh_Forward}
  )
  n64 = Transform(None=Constant3Vector() {constant=<rgba 0 0 1 0>}) {transformSourceType=TRANSFORMSOURCE_Local}
  n65 = ScalarParameter() {group=WPO_Positioning, parameterName=Push_Mesh_Up, sortPriority=1}
  n66 = Constant3Vector() {constant=<rgba 1 0 0 0>}
  n67 = Add(
   A=Add(A=Add(A=@n62, B=@n63), B=Multiply(A=@n64, B=@n65))
   B=Multiply(
    A=Transform(None=@n66) {transformSourceType=TRANSFORMSOURCE_Local}
    B=ScalarParameter() {group=WPO_Positioning, parameterName=Push_Mesh_Left/Right}
   )
  )
  MaterialFunctionCall!(Material=MakeMaterialAttributes!(
    BaseColor
    Metallic
    Specular
    Roughness
    Anisotropy
    EmissiveColor=MaterialFunctionCall(Emissive=@n45) {materialFunction=&MF_FX_EmissiveAdjustments}.Result
    Opacity=Multiply(A=@n46.A, B=@n47)
    OpacityMask
    Normal
    Tangent
    WorldPositionOffset=Add(A=@n48.Camera Offset, B=@n67)
    SubsurfaceColor
    ClearCoat
    ClearCoatRoughness
    AmbientOcclusion
    Refraction
    CustomizedUVs_0
    CustomizedUVs_1
    CustomizedUVs_2
    CustomizedUVs_3
    CustomizedUVs_4
    CustomizedUVs_5
    CustomizedUVs_6
    CustomizedUVs_7
    PixelDepthOffset
    ShadingModel
    Displacement
   )) {materialFunction=/Game/Packages/Fortress_SharedMaterials/Base_Material_Functions/BaseCharacterShaderFunctions/BaseFN_CharacterOpacity}
 ]
}
